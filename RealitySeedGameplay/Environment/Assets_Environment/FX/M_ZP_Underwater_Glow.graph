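# Masked, lit particle material for an underwater glow effect, usable on Niagara sprites and mesh particles.
# A panning water normal varies a sphere mask's radius, which combines with a radial gradient and particle alpha to drive dithered opacity thresholds.
# The same normal offsets reflection-map and sphere-normal UVs, while mobile keeps plain UVs.
# The squared reflection sample is scaled by an Emissive parameter and tinted by particle color.
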Material /RealitySeedGameplay/Environment/Assets_Environment/FX/M_ZP_Underwater_Glow
graph {
 properties {
  bTangentSpaceNormal=false
  bUsedWithMeshParticles=true
  bUsedWithNiagaraMeshParticles=true
  bUsedWithNiagaraSprites=true
  blendMode=BLEND_Masked
  ditherOpacityMask=true
  pixelDepthOffsetMode=PDOM_Legacy
  translucencyLightingMode=TLM_VolumetricPerVertexNonDirectional
  usedShadingModels=Default Lit
 }
 nodes [
  Constant()
  Constant() {r=0.5}
  LinearInterpolate(
   A=VectorParameter() {parameterName=SSS, defaultValue=<rgba 0.1675 0.3281 0.3281 1>}.RGB
   Alpha=TextureSample() {texture=/Game/Effects/Fort_Effects/Textures/Gameplay/T_Sphere_Mask}.R
  ) {constB=0.1}
  n1 = ParticleColor()
  n2 = TextureCoordinate()
  n3 = Constant() {r=0.5}
  n4 = TextureCoordinate()
  n5 = ScalarParameter() {defaultValue=1, parameterName=ScaleUVs}
  n6 = ScalarParameter() {parameterName=water_SpeedX}
  n7 = ScalarParameter() {defaultValue=0.1, parameterName=water_SpeedY}
  n8 = TextureSample(UVs=Panner(
    Coordinate=MaterialFunctionCall(UVs=@n4, Texture Scale=@n5) {materialFunction=&ScaleUVsByCenter}.UVs
    Speed=AppendVector(A=@n6, B=@n7)
   )) {samplerType=SAMPLERTYPE_Normal, texture=/Game/Packages/Fortnite_UniqueMaterials_1/Textures/T_Water_TilingNormal_Waves_01}
  n9 = Add(A=@n8.R, B=@n8.R)
  n10 = ScalarParameter() {defaultValue=0.5, parameterName=MaskRadius}
  n11 = LinearInterpolate(Alpha=Saturate(None=LinearInterpolate(B=@n9, Alpha=@n10) {constA=1})) {constA=0.1, constB=0.5}
  n12 = ScalarParameter() {defaultValue=100, parameterName=MaskHardness}
  n13 = MaterialFunctionCall(UVs=@n4, Radius=Constant() {r=0.5}, Density=ScalarParameter() {defaultValue=8, parameterName=RadialMaskDensity}) {materialFunction=&RadialGradientExponential}
  n14 = Multiply(A=@n1.A, B=Multiply(A=SphereMask(A=@n2, B=@n3, Radius=@n11, Hardness=@n12), B=@n13.RadialGradientExponential))
  MaterialFunctionCall(Alpha Threshold=@n14) {materialFunction=&DitherTemporalAA_Fortnite}
  MaterialFunctionCall(Alpha Threshold=@n14) {materialFunction=&DitherTemporalAA}
  MaterialFunctionCall(Emissive=@n1.RGB) {materialFunction=&MF_FX_EmissiveAdjustments}
  n15 = ShadingPathSwitch(Default=Add(A=TextureCoordinate(), B=ComponentMask(None=@n8.RGB) {g=true, r=true}), Mobile=TextureCoordinate())
  n16 = TextureSample(UVs=ComponentMask(None=@n15) {g=true, r=true}) {texture=/Game/Packages/Fortress_SharedMaterials/Textures/ReflectionMap2}
  Multiply(A=Multiply(A=ScalarParameter() {defaultValue=0.5, parameterName=Emissive}, B=Multiply(A=@n16.RGB, B=@n16.RGB)), B=@n1.RGB)
  Multiply(A=@n13.RadialGradientExponential)
  TextureSample() {samplerType=SAMPLERTYPE_Normal, texture=/Game/Packages/Fortnite_UniqueMaterials_1/Textures/T_Water_N}
  TextureSample(UVs=@n15) {samplerType=SAMPLERTYPE_Normal, texture=&T_Sphere_N}
 ]
}
