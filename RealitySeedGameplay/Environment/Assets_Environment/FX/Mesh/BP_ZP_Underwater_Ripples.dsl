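; Blueprint environment FX actor that, going by its name, provides underwater ripples built from twelve ripple mesh components.
; The construction script lowers the OpacityMult material parameter on ripples 01 and 03.
; It then loops over all twelve meshes, randomizing XTileOffset and the MainTexPan vector on their materials so tiling and panning vary per mesh.
; BeginPlay, Tick and overlap events are empty, so the Blueprint graph adds no runtime logic.
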(event EventBeginPlay)

(event Collision|EventActorBeginOverlap
  (params (OtherActor :type Actor%20Object%20Reference)))

(event EventTick
  (params (DeltaSeconds :type Float%20%28single-precision%29)))

(fn ConstructionScript
  (Rendering|Material|SetScalarParameterValueonMaterials (Variables|Default|GetSM_ZP_Ripple_01) "OpacityMult" 0.25)
  (Rendering|Material|SetScalarParameterValueonMaterials (Variables|Default|GetSM_ZP_Ripple_03) "OpacityMult" 0.5)
  (for _array_element (Utilities|Array|MakeArray (Variables|Default|GetSM_ZP_Ripple_01) (Variables|Default|GetSM_ZP_Ripple_02) (Variables|Default|GetSM_ZP_Ripple_03) (Variables|Default|GetSM_ZP_Ripple_04) (Variables|Default|GetSM_ZP_Ripple_05) (Variables|Default|GetSM_ZP_Ripple_06) (Variables|Default|GetSM_ZP_Ripple_07) (Variables|Default|GetSM_ZP_Ripple_08) (Variables|Default|GetSM_ZP_Ripple_09) (Variables|Default|GetSM_ZP_Ripple_10) (Variables|Default|GetSM_ZP_Ripple_11) (Variables|Default|GetSM_ZP_Ripple_12))
    (Rendering|Material|SetScalarParameterValueonMaterials _array_element "XTileOffset" (Math|Random|RandomFloatinRange -1.0 1.0))
    (Rendering|Material|SetVectorParameterValueonMaterials _array_element "MainTexPan" (Math|Vector|MakeVector (Math|Random|RandomFloatinRange 0.01 0.005) (Math|Random|RandomFloatinRange 0.025 0.012)))))
