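# Translucent, two-sided unlit material for Zero Point ripple FX meshes, with separate top and river controls selected by vertex-color blue.
# Color is a time-animated hue shift masked by vertex red, with desaturation and time-of-day brightness.
# Opacity blends two panned wave textures along UV1, multiplied by a scrolling line mask and UV edge gradients.
# Quality and shading-path switches add camera depth fade on higher tiers and inverted Fresnel fade on Epic.
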Material /RealitySeedGameplay/Environment/Assets_Environment/FX/Mesh/M_ZP_Ripple
graph {
 properties {blendMode=BLEND_Translucent, pixelDepthOffsetMode=PDOM_Legacy, shadingModel=MSM_Unlit, twoSided=true, usedShadingModels=Unlit}
 nodes [
  Constant() {r=4}
  n1 = ScalarParameter() {defaultValue=1, parameterName="Value_Morning (S)"}
  n2 = ScalarParameter() {defaultValue=1, parameterName="Value_Day (S)"}
  n3 = ScalarParameter() {defaultValue=1, parameterName="Value_Evening (S)"}
  n4 = ScalarParameter() {defaultValue=1, parameterName="Value_Night (S)"}
  n5 = ScalarParameter() {defaultValue=1, parameterName=MultColor_Top}
  n6 = ScalarParameter() {defaultValue=1, parameterName=MultColor_River}
  n7 = VertexColor()
  n8 = TextureCoordinate() {coordinateIndex=1}
  n9 = Time() {period=2}
  n10 = Sine(None=Frac(None=Multiply(A=@n9) {constB=0.1}))
  n11 = Multiply(A=Add(A=ComponentMask(None=@n8) {g=true}, B=@n10), B=@n10)
  n12 = ScalarParameter() {defaultValue=1, parameterName=HueShiftMult}
  n13 = MaterialFunctionCall(
   Hue Shift Percentage=Multiply(A=VertexColor().R, B=Multiply(A=@n11, B=@n12))
   Texture=VectorParameter() {parameterName=BaseColor, defaultValue=<rgba 5 2 1 1>}.RGB
  ) {materialFunction=&HueShift}
  n14 = ScalarParameter() {parameterName=Desaturation}
  Multiply(
   A=VertexInterpolator(VS=MaterialFunctionCall(Value_Morning=@n1, Value_Day=@n2, Value_Evening=@n3, Value_Night=@n4) {materialFunction=&Scalar_Time_of_Day}.Result).PS
   B=Multiply(A=LinearInterpolate(A=@n5, B=@n6, Alpha=@n7.B), B=Desaturation(None=@n13.Result, Fraction=@n14))
  )
  n15 = ScalarParameter() {parameterName=XTileOffset}
  n16 = Multiply(
   A=TextureCoordinate() {uTiling=0.2, vTiling=4}
   B=ComponentMask(None=VectorParameter() {parameterName=TileMult, defaultValue=<rgba 1 1 0 1>}.RGB) {g=true, r=true}
  )
  n17 = ComponentMask(None=@n16) {r=true}
  n18 = ComponentMask(None=@n16) {g=true}
  n19 = ComponentMask(None=TextureCoordinate() {coordinateIndex=1}) {g=true}
  n20 = Multiply(A=ComponentMask(None=VectorParameter() {parameterName=TileMult, defaultValue=<rgba 1 1 0 1>}.RGB) {g=true}, B=@n19)
  n21 = Panner(
   Coordinate=AppendVector(A=Add(A=@n15, B=@n17), B=Add(A=@n18, B=@n20))
   Speed=VectorParameter() {parameterName=MainTexPan, defaultValue=<rgba 0 -0.1 0 1>}.RGB
  )
  n22 = Panner(Coordinate=TextureCoordinate() {coordinateIndex=1}) {speedY=0.1}
  n23 = ComponentMask(None=VectorParameter() {parameterName=TileMultMask, defaultValue=<rgba 1 1 0 1>}.RGB) {g=true, r=true}
  n24 = Multiply(
   A=LinearInterpolate(
    A=TextureSampleParameter2D(UVs=@n21) {parameterName=MainTexture_01, texture=/Game/Athena/Environments/Nexus/Textures/T_ZeroPoint_MagneticWaves}.G
    B=TextureSampleParameter2D(UVs=@n21) {parameterName=MainTexture_02, texture=&T_ZeroPoint_MagneticWaves}.G
    Alpha=@n19
   )
   B=TextureSampleParameter2D(UVs=Multiply(A=@n22, B=@n23)) {parameterName=MaskTexture, texture=/RealitySeedGameplay/Environment/Assets_Daisy/FX/Textures/T_Lines_Base_01}.G
  )
  n25 = MaterialFunctionCall(UV Channel=TextureCoordinate()) {materialFunction=&LinearGradient}
  n26 = OneMinus(None=@n25.UGradient)
  n27 = ScalarParameter() {defaultValue=1, parameterName=Mult_Mask_Top}
  n28 = ScalarParameter() {defaultValue=1, parameterName=Mult_Mask_River}
  n29 = VertexColor()
  n30 = Power(Base=Multiply(A=Multiply(A=@n25.UGradient, B=@n26), B=LinearInterpolate(A=@n27, B=@n28, Alpha=@n29.B))) {constExponent=4}
  n31 = MaterialFunctionCall(UV Channel=TextureCoordinate()) {materialFunction=&LinearGradient}
  n32 = Multiply(A=Multiply(A=@n31.VGradient, B=OneMinus(None=@n31.VGradient)), B=Constant() {r=4})
  n33 = Multiply(
   A=ScalarParameter() {defaultValue=1, parameterName=OpacityMult}
   B=Multiply(A=Multiply(A=@n24, B=@n30), B=Power(Base=@n32) {constExponent=4})
  )
  n34 = Multiply(
   A=@n33
   B=MaterialFunctionCall(
    Fade Length=ScalarParameter() {defaultValue=512, parameterName=NearFadeDistance}
    Fade Offset=ScalarParameter() {defaultValue=256, parameterName=NearFadeDistance_Start_Offset}
   ) {materialFunction=&CameraDepthFade}.Result
  )
  n35 = ShadingPathSwitch(Default=@n34, Forward=@n33, Mobile=@n33)
  n36 = Fresnel(ExponentIn=ScalarParameter() {defaultValue=0.5, parameterName=FresnelFade_Exponent}) {baseReflectFraction=0.01, exponent=0.5}
  QualitySwitch(Default=@n35, Low=@n33, High=@n35, Medium=@n33, Epic=Multiply(A=OneMinus(None=@n36), B=@n34))
  VertexColor()
 ]
}
