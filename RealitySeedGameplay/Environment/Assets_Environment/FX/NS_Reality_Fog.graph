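# Niagara system for large, tinted GPU sprite fog over terrain; its only emitter, ChromaticFog, is disabled.
# A terrain-near-player spawn module and flattened sphere location place particles, with spawn count scaled by distance.
# Camera-distance curves drive lifetime, alpha and size, fading fog both near and far.
# Time of day blends random purple-to-blue color ranges and sets brightness, while dynamic material parameters add per-particle variation.
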NiagaraSystem /RealitySeedGameplay/Environment/Assets_Environment/FX/NS_Reality_Fog
dependencies {
 usedDataInterfaces [&NiagaraDataInterfaceCurve, &NiagaraDataInterfaceCamera, &NiagaraDataInterfaceVector2DCurve]
 usedDynamicInputs []
 usedModules [&SystemState]
}
emitters {ChromaticFog {
  bEnabled=false
  simTarget=GPUComputeSim
  emitterUpdateScript [
   &EmitterState(
    Enable Distance Culling=true
    MaxDistance=35000
    Module.bApplySpawnCountScaleByDistance=true
    Module.bUseMaxDistance=true
    Module.bUseMinDistance=false
    Scalability Mode=::NewEnumerator1
    Scale Spawn Count=true
   )
   &SpawnRate(SpawnRate=8, Use Spawn Probability=false)
   /SetVariables(Emitter.CameraPosition:Vector3f=&ReturnCameraVector(Camera Property=::NewEnumerator0, Coordinate Space=::Simulation)) {isSetParameters=true}
  ]
  particleSpawnScript [
   &InitializeParticle(
    Lifetime=12
    Sprite Size Mode=::NewEnumerator1
    Uniform Sprite Size Max=4000
    Uniform Sprite Size Min=3000
    UsePositionOffset=false
    Write Material Random=true
   )
   /NiagaraFN/Modules/Spawn/FN_SpawnParticlesOnTerrainNearPlayer(
    Height Texture -x -y:Vector3f=<-130048 -130048 0>
    HeightTexture +x +y:Vector3f=<130048 130048 0>
    Player Position:NiagaraPosition=$Particles.Position
   )
   &ShapeLocation(MaskSpawnGroup=false, Non Uniform Scale=<1 1 0>, Override Seed=false, Sphere Radius=5000, UseOwnerScale=true)
   &CameraOffset(Camera Offset Amount=500)
  ]
  particleUpdateScript [
   &ParticleState(Lifetime:NiagaraFloat=&Multiply_Float(
     A=$Particles.Lifetime
     B=&FloatFromCurve(
      FloatCurve=DataInterface() {propertyValues {Curve {keys [
          {interpMode=RCIM_Cubic, tangentMode=RCTM_User, time=0.05, value=0}
          {arriveTangent=0.0074, interpMode=RCIM_Cubic, leaveTangent=0.0074, tangentMode=RCTM_User, time=0.25, value=1}
          {arriveTangent=-0.0952, interpMode=RCIM_Cubic, leaveTangent=-0.0952, tangentMode=RCTM_User, time=0.75, value=1}
          <at 0.95 0>
         ]}}}
      CurveIndex=&NormalizeDistanceRange(
       Start Position:NiagaraPosition=Linked() {linkedVariable {name=Emitter.CameraPosition, type=Vector3f}}
       End Position:NiagaraPosition=$Particles.Position
       Distance:NiagaraFloat=10000
       Invert Normalized Range:NiagaraBool=false
      )
     )
    ))
   /SetVariables(Particles.LerpTODColor:LinearColor=&LerpLinearColors(
     StartColor=&RandomRangeLinearColor(Minimum=<rgba 0.7057 0.4115 1 1>, Maximum=<rgba 0.0852 0.0385 0.1719 1>)
     EndColor=&RandomRangeLinearColor(Minimum=<rgba 0.2188 0.2573 1 1>, Maximum=<rgba 0.0669 0.0614 0.1719 1>)
     LerpFactor=&FNTimeOfDayScalar(Night Scale=1, Morning Scale=1, Evening Scale=1)
    )) {isSetParameters=true}
   &ScaleColor(
    Scale Alpha=&FloatFromCurve(
     FloatCurve=DataInterface() {propertyValues {Curve {keys [
         <at 0 0 RCIM_Cubic>
         {arriveTangent=0.1469, interpMode=RCIM_Cubic, leaveTangent=0.1469, tangentMode=RCTM_User, time=0.2, value=1.125}
         <at 0.85 0>
         <at 0.9 0 RCIM_Cubic>
        ]}}}
     Scale Curve=&FloatFromCurve(
      FloatCurve=DataInterface() {propertyValues {Curve {keys [
          {interpMode=RCIM_Cubic, tangentMode=RCTM_User, time=0.15, value=0}
          {arriveTangent=0.0074, interpMode=RCIM_Cubic, leaveTangent=0.0074, tangentMode=RCTM_User, time=0.25, value=0.9932}
          {arriveTangent=-0.0952, interpMode=RCIM_Cubic, leaveTangent=-0.0952, tangentMode=RCTM_User, time=0.8, value=1}
          <at 0.95 0>
         ]}}}
      CurveIndex=&NormalizeDistanceRange(
       Start Position:NiagaraPosition=Linked() {linkedVariable {name=Emitter.CameraPosition, type=Vector3f}}
       End Position:NiagaraPosition=$Particles.Position
       Distance:NiagaraFloat=25000
       Invert Normalized Range:NiagaraBool=false
      )
     )
    )
    Scale RGB=&Multiply_VectorByFloat(
     Vector=&MakeVectorFromLinearColorRGB(Color=$Particles.LerpTODColor)
     Float=&FNTimeOfDayScalar(Night Scale=0.1, Morning Scale=0.25, Day Scale=0.5, Evening Scale=0.35)
    )
    ScaleA=true
    ScaleRGB=true
   )
   &ScaleSpriteSize(Scale Factor=&Vector2DFromCurve(
     Vector2Curve=DataInterface() {propertyValues {
       XCurve {keys [
         {interpMode=RCIM_Cubic, tangentMode=RCTM_User, time=0.3581, value=1}
         {arriveTangent=-1.514, interpMode=RCIM_Cubic, leaveTangent=-1.514, tangentMode=RCTM_User, time=1, value=0.5}
        ]}
       YCurve {keys [
         {interpMode=RCIM_Cubic, tangentMode=RCTM_User, time=0.3581, value=1}
         {arriveTangent=-1.514, interpMode=RCIM_Cubic, leaveTangent=-1.514, tangentMode=RCTM_User, time=1, value=0.5}
        ]}
      }}
     ScaleCurve=&Multiply_Vector2(A=<1 1>, B=<1 1>)
    )) {module=ScaleSpriteSize001}
   &ScaleSpriteSize(Scale Factor=&Multiply_Vector2(
     A=<1 1>
     B=&Vector2DFromFloat(Value=&Multiply_Float(
       A=&FloatFromCurve(FloatCurve=DataInterface() {propertyValues {Curve {keys [{arriveTangent=-1.2812, interpMode=RCIM_Cubic, leaveTangent=-1.2812, tangentMode=RCTM_User, time=0.9, value=1}, <at 1 0>]}}})
       B=&FloatFromCurve(
        FloatCurve=DataInterface() {propertyValues {Curve {keys [
            {interpMode=RCIM_Cubic, tangentMode=RCTM_User, time=0, value=0.0005}
            {arriveTangent=1.3922, leaveTangent=1.3922, time=0.4, value=1}
            <at 0.8 1>
            <at 0.9 0>
           ]}}}
        CurveIndex=&NormalizeDistanceRange(
         Start Position:NiagaraPosition=Linked() {linkedVariable {name=Emitter.CameraPosition, type=Vector3f}}
         End Position:NiagaraPosition=$Particles.Position
         Distance:NiagaraFloat=25000
         Invert Normalized Range:NiagaraBool=false
        )
       )
      ))
    ))
   &DynamicMaterialParameters(
    Index 0 Param 1=1
    Index 0 Param 2=1
    Index 0 Param 3=1
    Index 0 Param 4=1
    Index 1 Param 1=&RandomRangeFloat(Minimum=0.03, Maximum=0.07)
    Index 1 Param 2=&RandomRangeFloat(Minimum=0.1, Maximum=0.2)
    Index 1 Param 3=&RandomRangeFloat(Minimum=0, Maximum=1)
    Index 1 Param 4=&RandomRangeFloat(Minimum=0, Maximum=1)
    Index 2 Param 1=-0.02
    Index 2 Param 2=0.03
    Index 2 Param 3=1.7
    Index 2 Param 4=2
    Index 3 Param 1=15000
    Index 3 Param 2=1
    Index 3 Param 3=&RandomRangeFloat(Minimum=0.05, Maximum=0.2)
    Module.Param0WriteEnabled=true
    Module.Param0WriteEnabled1=true
    Module.Param0WriteEnabled2=true
    Module.Param0WriteEnabled3=true
    Module.Param1WriteEnabled=true
    Module.Param1WriteEnabled1=true
    Module.Param1WriteEnabled2=true
    Module.Param1WriteEnabled3=true
    Module.Param2WriteEnabled=true
    Module.Param2WriteEnabled1=true
    Module.Param2WriteEnabled2=true
    Module.Param2WriteEnabled3=true
    Module.Param3WriteEnabled=true
    Module.Param3WriteEnabled1=true
    Module.Param3WriteEnabled2=true
    Module.Param3WriteEnabled3=true
    Write Parameter Index 1=true
    Write Parameter Index 2=true
    Write Parameter Index 3=true
   )
  ]
  rendererClasses [&NiagaraSpriteRendererProperties]
  renderers [{rendererClass=&NiagaraSpriteRendererProperties}]
 }}
summary {
 systemName=NS_Reality_Fog
 emitters [{bEnabled=false, emitterName=ChromaticFog, simTarget=GPUComputeSim, rendererClasses [&NiagaraSpriteRendererProperties]}]
 userVariables [User.Landscape {
   type=NiagaraDataInterfaceLandscape
   description=""
   defaultValue {
    struct=NiagaraExt_VariableValue_DataInterface
    value {
     dataInterface=&self:NiagaraDataInterfaceLandscape_0
     dataInterfaceClass=&NiagaraDataInterfaceLandscape
    }
   }
  }]
}
